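# M_SimpleTentacleMat is a Default Lit material for the tentacle sample quest prop, layering goop over the base tentacle textures.
# The Slice_TechArt alpha masks the blend to a goop color and roughness over the diffuse and the mask map's blue channel.
# The same alpha blends in a goop normal made from flattened base and tiled water normals via BlendAngleCorrectedNormals.
# The slice texture's blue channel separately masks in the Sac Int color.
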Material /Hermes/Props/Quest/Quest_TentacleSample_A/Materials/M_SimpleTentacleMat
graph {
 properties {bAutomaticallySetUsageInEditor=false, usedShadingModels=Default Lit}
 nodes [
  n1 = TextureSample() {texture=/Hermes/Props/Quest/Quest_TentacleSample_A/Textures/Slice_TechArt}
  LinearInterpolate(B=VectorParameter() {parameterName=Sac Int, defaultValue=<rgba 0.1146 0.0265 0.0649 1>}.RGB, Alpha=@n1.B)
  LinearInterpolate(
   A=TextureSample() {texture=/Hermes/Props/Quest/Quest_TentacleSample_A/Textures/T_Quest_TentacleSample_A_A_D}.RGB
   B=VectorParameter() {parameterName=GoopColor, defaultValue=<rgba 0.1146 0.006 0.0361 1>}.RGB
   Alpha=@n1.A
  )
  LinearInterpolate(
   A=TextureSample() {samplerType=SAMPLERTYPE_Masks, texture=/Hermes/Props/Quest/Quest_TentacleSample_A/Textures/T_Quest_TentacleSample_A_A_S}.B
   B=ScalarParameter() {parameterName=GoopRoughness}
   Alpha=@n1.A
  )
  n2 = TextureSample() {samplerType=SAMPLERTYPE_Normal, texture=/Hermes/Props/Quest/Quest_TentacleSample_A/Textures/T_Quest_TentacleSample_A_A_N}
  n3 = Constant() {r=0.8}
  n4 = TextureSample(UVs=TextureCoordinate() {uTiling=4, vTiling=4}) {samplerType=SAMPLERTYPE_Normal, texture=/FortWater/Textures/T_Water_N}
  n5 = Constant() {r=0.5}
  LinearInterpolate(
   A=@n2.RGB
   B=MaterialFunctionCall(
    BaseNormal=MaterialFunctionCall(Normal=@n2.RGB, Flatness=@n3) {materialFunction=&FlattenNormal}.Result
    AdditionalNormal=MaterialFunctionCall(Normal=@n4.RGB, Flatness=@n5) {materialFunction=&FlattenNormal}.Result
   ) {materialFunction=&BlendAngleCorrectedNormals}.Result
   Alpha=@n1.A
  )
 ]
}
